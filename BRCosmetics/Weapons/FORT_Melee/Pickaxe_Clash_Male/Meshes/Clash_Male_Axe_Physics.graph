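# Physics asset for the Clash Male pickaxe, a cosmetic melee weapon, giving its skeletal mesh a simple collision shape.
# It holds a single body on the lowerarm_r bone, made of one wide capsule that is offset and rotated from the bone's origin.
# It has no constraints, so it defines no jointed or multi-body simulation.
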PhysicsAsset /BRCosmetics/Weapons/FORT_Melee/Pickaxe_Clash_Male/Meshes/Clash_Male_Axe_Physics
bodies {lowerarm_r [{shapeType=Capsule, length=36.6391, radius=34.4592, center=<-36.8814 -2.9999 -4.1918>, rotation=<rot 0 91.0914 -85.8247>}]}
constraints []
